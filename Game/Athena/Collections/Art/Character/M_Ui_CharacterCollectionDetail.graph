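# Translucent UI material that draws a character's outfit art for the collection detail display, with known, unknown and undiscovered looks.
# Fitted UVs get an Intro-driven scale, offset and fade, plus a scale correction chosen by checking texture width against 1024.
# A static switch swaps the icon, tinted by discovery state and a vertical color gradient, for a faint flat-colored unknown silhouette.
# A vertical falloff and optional angled side clipping shape opacity.
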Material /Game/Athena/Collections/Art/Character/M_Ui_CharacterCollectionDetail
graph {
 properties {blendMode=BLEND_Translucent, materialDomain=MD_UI, pixelDepthOffsetMode=PDOM_Legacy, usedShadingModels=Default Lit}
 nodes [
  n1 = NamedRerouteUsage()
  n2 = ComponentMask(None=@n1) {g=true}
  n3 = LinearInterpolate(
   A=ScalarParameter() {defaultValue=-0.125, parameterName=IntroOffset}
   Alpha=ScalarParameter() {defaultValue=1, parameterName=Intro, sliderMax=1}
  ) {constA=0.5, constB=0}
  n4 = MaterialFunctionCall(
   UVs=AppendVector(A=ComponentMask(None=@n1) {r=true}, B=Add(A=@n2, B=@n3) {constB=0})
   Texture Scale=LinearInterpolate(A=ScalarParameter() {defaultValue=0.75, parameterName=IntroSize}, Alpha=ScalarParameter() {defaultValue=1, parameterName=Intro, sliderMax=1}) {constA=0.5}
  ) {materialFunction=&ScaleUVsByCenter}
  n5 = ScalarParameter() {defaultValue=0.75, parameterName=512TextureSizeOffest_TOREMOVE}
  n6 = NamedRerouteUsage()
  n7 = ScalarParameter() {parameterName=Disable512Scaling}
  n8 = TextureSample(
   UVs=LinearInterpolate(
    A=MaterialFunctionCall(UVs=@n4.UVs, Texture Scale=@n5) {materialFunction=&ScaleUVsByCenter}.UVs
    B=@n4.UVs
    Alpha=LinearInterpolate(A=@n6, Alpha=@n7)
   )
   Tex=NamedRerouteUsage()
  ) {texture=/Game/UI/Foundation/Textures/BattleRoyale/FeaturedItems/Outfit/T-AthenaSoldiers-Character-PizzaParty-Yellow}
  n9 = Multiply(A=@n8.A, B=LinearInterpolate(Alpha=ScalarParameter() {defaultValue=1, parameterName=Intro, sliderMax=1}))
  n10 = ComponentMask(None=Add(A=NamedRerouteUsage(), B=ScalarParameter() {parameterName=Mask offset})) {g=true}
  n11 = ScalarParameter() {defaultValue=1, parameterName=Mask Falloff}
  n12 = OneMinus(None=Saturate(None=Multiply(A=@n10, B=@n11)))
  n13 = TextureObjectParameter() {
   parameterName=UnknownIcon
   texture=/Game/UI/Foundation/Textures/BattleRoyale/FeaturedItems/Outfit/T-AthenaSoldiers-CID-883-Athena-Commando-M-ChOneJonesy
  }
  n14 = Multiply(
   A=@n12
   B=TextureSample(UVs=NamedRerouteUsage(), Tex=@n13) {texture=&T-AthenaSoldiers-CID-883-Athena-Commando-M-ChOneJonesy}.A
  )
  n15 = ScalarParameter() {defaultValue=0.25, parameterName=UnknownOpacity, sliderMax=1}
  n16 = NamedRerouteUsage()
  n17 = ComponentMask(None=@n16) {r=true}
  n18 = Step(Y=Add(A=OneMinus(None=@n17), B=ScalarParameter() {defaultValue=0.216, parameterName=ClippingLeftMaskScale}) {constB=0.225})
  n19 = Rotator(Coordinate=@n16, Time=Divide(A=ScalarParameter() {defaultValue=-2.5, group=Degrees, parameterName=RightAngle}) {constB=360}) {centerX=0, speed=6.2832}
  n20 = ScalarParameter() {defaultValue=0.206, group=Offset, parameterName=ClippingRightMaskScale}
  n21 = Step(Y=ComponentMask(None=Add(A=@n19, B=@n20)) {r=true})
  n22 = ScalarParameter() {defaultValue=1, parameterName=ClippingMask}
  Multiply(
   A=StaticSwitchParameter(True=Multiply(A=@n9, B=@n12), False=Multiply(A=@n14, B=@n15)) {defaultValue=true, parameterName=is known}
   B=LinearInterpolate(B=Min(A=@n18, B=@n21), Alpha=Floor(None=@n22)) {constA=1}
  )
  n23 = TextureObjectParameter(UVs=NamedRerouteUsage()) {parameterName=ItemIcon, texture=&T-AthenaSoldiers-Character-PizzaParty-Yellow}
  n24 = TextureProperty(Tex=@n23)
  NamedRerouteDeclaration(None=MaterialFunctionCall!(A=ComponentMask(None=@n24) {r=true}, B=Constant() {r=1024}) {materialFunction=/FortUILibrary/MaterialFunctions/Basics/MF_UI_BooleanOperations}.IsGreaterOrEqual) {name=is512ALogic, nodeColor=<rgba 0 0 0 1>}
  NamedRerouteDeclaration(None=@n23) {name=Texture, nodeColor=<rgba 0 0 0 1>}
  NamedRerouteDeclaration(None=MaterialFunctionCall(ScalePivotPoint=Constant2Vector() {g=1, r=0.5}) {materialFunction=&MF_UI_UVScaling}.Scaled to Fit) {name=UV, nodeColor=<rgba 0.8047 0 1 1>}
  n25 = VectorParameter() {parameterName=RGB_Undiscovered, defaultValue=<rgba 0 0.0012 0.0262 1>}
  n26 = ScalarParameter() {defaultValue=1, parameterName=IsDiscovered, sliderMax=1}
  n27 = ComponentMask(None=Add(A=NamedRerouteUsage(), B=ScalarParameter() {parameterName=Color offset})) {g=true}
  n28 = ScalarParameter() {parameterName=Color Falloff}
  n29 = VectorParameter() {parameterName=RGBA, defaultValue=<rgba 0 0.0043 0.0833 0.536>}
  n30 = Subtract(A=Saturate(None=Multiply(A=@n27, B=@n28)), B=OneMinus(None=@n29.A)) {constB=0}
  StaticSwitchParameter(
   True=LinearInterpolate(A=LinearInterpolate(A=@n25.RGB, B=@n8.RGB, Alpha=@n26), B=@n29.RGB, Alpha=Saturate(None=@n30) {desc=Fixes Colors inverting})
   False=VectorParameter() {parameterName=RGB_Undiscovered, defaultValue=<rgba 0 0.0012 0.0262 1>}.RGB
  ) {defaultValue=true, parameterName=is known}
 ]
}
