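# Translucent UI material for the background of a character collection entry, with IsDiscovered switching palettes and IsKnown fading from a flat unknown color.
# An aspect-corrected screen-space triangle grid, rotated and panning, blends the base color toward the glow color.
# An exponential radial gradient, corrected for widget aspect via pixel size, adds more glow.
# A vertical UV gradient blends that result with the top gradient color.
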Material /Game/Athena/Collections/Art/Character/M_Ui_CharacterCollectionEntry_Background
graph {
 properties {blendMode=BLEND_Translucent, materialDomain=MD_UI, pixelDepthOffsetMode=PDOM_Legacy, usedShadingModels=Default Lit}
 nodes [
  n1 = VectorParameter() {parameterName=RGBA Unknown, defaultValue=<rgba 0 0.0012 0.0262 0.33>}
  LinearInterpolate(A=@n1.A, Alpha=ScalarParameter() {defaultValue=1, parameterName=IsKnown, sliderMax=1})
  n2 = VectorParameter() {parameterName=Background_topgradient_F, defaultValue=<rgba 0 0.004 0.0723 1>}
  n3 = VectorParameter() {group=background, parameterName=Background_topgradient_T, defaultValue=<rgba 0 0.0578 0.6038 1>}
  n4 = ScalarParameter() {group=control, parameterName=IsDiscovered, sliderMax=1}
  n5 = LinearInterpolate(
   A=VectorParameter() {group=background, parameterName=Background_topgradient_T}.RGB
   B=VectorParameter() {group=background, parameterName=Background_glow_T, defaultValue=<rgba 0 0.6654 1 1>}.RGB
   Alpha=ScalarParameter() {group=control, parameterName=IsDiscovered, sliderMax=1}
  )
  n6 = MaterialFunctionCall(Float2=MaterialFunctionCall() {materialFunction=&GetUserInterfaceUV}.Pixel Size) {materialFunction=&BreakOutFloat2Components}
  n7 = MaterialFunctionCall(Float2=TextureCoordinate()) {materialFunction=&BreakOutFloat2Components}
  n8 = Multiply(A=Divide(A=@n6.R, B=Max(A=@n6.G)), B=@n7.R)
  n9 = ScalarParameter() {defaultValue=0.5, group=background, parameterName=RadialGradient_u}
  n10 = ScalarParameter() {defaultValue=0.5, group=background, parameterName=RadialGradient_V}
  n11 = Saturate(None=MaterialFunctionCall(
    UVs=AppendVector(A=@n8, B=@n7.G)
    CenterPosition=AppendVector(A=@n9, B=@n10)
    Radius=ScalarParameter() {defaultValue=0.75, group=background, parameterName=Radial Gradient Radius_V}
    Density=ScalarParameter() {defaultValue=1, group=background, parameterName=Radial Gradient Density}
   ) {materialFunction=&RadialGradientExponential}.RadialGradientExponential)
  n12 = MaterialFunctionCall(Float2=MaterialFunctionCall() {materialFunction=&ScreenResolution}.Visible Resolution) {materialFunction=&BreakOutFloat2Components}
  n13 = Max(A=@n12.G)
  n14 = Multiply(
   A=AppendVector(A=Divide(A=@n12.R, B=@n13), B=Constant() {r=1})
   B=MaterialFunctionCall() {materialFunction=&ScreenAlignedUVs}."X 100%, Y 100%"
  )
  n15 = ScalarParameter() {defaultValue=-0.0278, parameterName=Tri Angle}
  n16 = ScalarParameter() {defaultValue=0.5, parameterName=PanSpeed}
  n17 = Constant()
  n18 = Multiply(
   A=Panner(Coordinate=Rotator(Coordinate=@n14, Time=@n15) {speed=6.2832}, Speed=AppendVector(A=@n16, B=@n17))
   B=ScalarParameter() {defaultValue=1, parameterName=Tri Tile Amount}
  )
  n19 = LinearInterpolate(
   A=LinearInterpolate(
    A=VectorParameter() {parameterName=Background_base_F, defaultValue=<rgba 0.0033 0.0185 0.3763 1>}.RGB
    B=VectorParameter() {group=background, parameterName=Background_base_T, defaultValue=<rgba 0 0.2159 1 1>}.RGB
    Alpha=ScalarParameter() {group=control, parameterName=IsDiscovered, sliderMax=1}
   )
   B=@n5
   Alpha=Multiply(
    A=TextureSample(UVs=@n18) {texture=/Game/Global/Textures/Default/Masks_and_Gradients/T-TriangleGrid}.R
    B=ScalarParameter() {defaultValue=1, parameterName=TriangleDensity, sliderMax=1}
   )
  )
  n20 = ComponentMask(None=TextureCoordinate()) {g=true}
  LinearInterpolate(
   A=@n1.RGB
   B=LinearInterpolate(A=LinearInterpolate(A=@n2.RGB, B=@n3.RGB, Alpha=@n4), B=LinearInterpolate(A=@n19, B=@n5, Alpha=@n11), Alpha=Saturate(None=@n20))
   Alpha=ScalarParameter() {defaultValue=1, parameterName=IsKnown, sliderMax=1}
  )
 ]
}
