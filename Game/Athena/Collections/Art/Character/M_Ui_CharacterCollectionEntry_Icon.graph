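# Translucent UI material for a character collection entry icon, driven by hover, Selected, IsDiscovered and IsComplete parameters.
# The aspect-corrected ItemIcon shifts and zooms on hover and blends with background gradient colors, with a gold tint outside it when discovered and complete.
# Four slightly rotated, sharpened edge masks form a skewed border that shifts on hover and brightens when selected.
# A location-marker texture is also sampled at a Selected-driven position, with no visible downstream use.
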Material /Game/Athena/Collections/Art/Character/M_Ui_CharacterCollectionEntry_Icon
graph {
 properties {blendMode=BLEND_Translucent, materialDomain=MD_UI, pixelDepthOffsetMode=PDOM_Legacy, usedShadingModels=Default Lit}
 nodes [
  n1 = MaterialFunctionCall(Float2=MaterialFunctionCall() {materialFunction=&GetUserInterfaceUV}.Pixel Size) {materialFunction=&BreakOutFloat2Components}
  n2 = Max(A=@n1.G)
  n3 = MaterialFunctionCall(Float2=TextureCoordinate() {coordinateIndex=4}) {materialFunction=&BreakOutFloat2Components}
  n4 = AppendVector(A=Multiply(A=Divide(A=@n1.R, B=@n2), B=@n3.R), B=@n3.G)
  n5 = AppendVector(
   A=ScalarParameter() {defaultValue=0.09, group=Item, parameterName=Itemposition_U}
   B=LinearInterpolate(
    A=ScalarParameter() {defaultValue=0.04, group=Item, parameterName=Itemposition_V}
    B=ScalarParameter() {defaultValue=0.04, group=Item, parameterName=Itemposition_V_Hover}
    Alpha=ScalarParameter() {group=control, parameterName=hover, sliderMax=1, sliderMin=0.1}
   ) {constA=1, constB=0}
  )
  n6 = ScalarParameter() {defaultValue=1.01, parameterName=hoverZoom}
  n7 = ScalarParameter() {group=control, parameterName=hover, sliderMax=1, sliderMin=0.1}
  n8 = TextureSampleParameter2D(UVs=MaterialFunctionCall(UVs=Add(A=@n4, B=@n5), Texture Scale=LinearInterpolate(B=@n6, Alpha=@n7) {constA=1, constB=1.01}) {materialFunction=&ScaleUVsByCenter}.UVs) {parameterName=ItemIcon, texture=/Game/UI/Foundation/Textures/Icons/Heroes/Athena/Soldier/T-Soldier-HID-005-Athena-Commando-M-L}
  n9 = LinearInterpolate(
   B=LinearInterpolate(Alpha=ScalarParameter() {group=control, parameterName=IsComplete, sliderMax=1}) {constB=0}
   Alpha=ScalarParameter() {group=control, parameterName=IsDiscovered, sliderMax=1}
  ) {constB=0}
  n10 = Add(A=@n8.A, B=@n9)
  n11 = Rotator(
   Coordinate=TextureCoordinate() {coordinateIndex=4}
   Time=LinearInterpolate(
    A=ScalarParameter() {defaultValue=0.002, group=Border Angles, parameterName=Top-Angle}
    B=ScalarParameter() {defaultValue=-0.002, group=Border Angles, parameterName=Top-Angle-hover}
    Alpha=ScalarParameter() {group=control, parameterName=hover, sliderMax=1}
   )
  ) {speed=6.2832}
  n12 = ScalarParameter() {defaultValue=0.035, group=Border_Range, parameterName=Top-RangeOffset, sliderMax=1}
  n13 = ScalarParameter() {defaultValue=0.035, group=Border_Range, parameterName=Top-RangeOffset-hover, sliderMax=1}
  n14 = ScalarParameter() {group=control, parameterName=hover, sliderMax=1}
  n15 = Multiply(
   A=Subtract(A=ComponentMask(None=@n11) {g=true}, B=LinearInterpolate(A=@n12, B=@n13, Alpha=@n14))
   B=ScalarParameter() {defaultValue=512, group=Border sharpness, parameterName=HorizontalSharpness}
  )
  n16 = Rotator(
   Coordinate=TextureCoordinate() {coordinateIndex=4}
   Time=LinearInterpolate(
    A=ScalarParameter() {defaultValue=-0.002, group=Border Angles, parameterName=Left-Angle}
    B=ScalarParameter() {defaultValue=0.001, group=Border Angles, parameterName=Left-Angle-hover}
    Alpha=ScalarParameter() {group=control, parameterName=hover, sliderMax=1}
   )
  ) {speed=6.2832}
  n17 = ScalarParameter() {defaultValue=0.035, group=Border_Range, parameterName=Left-RangeOffset, sliderMax=1}
  n18 = ScalarParameter() {defaultValue=0.035, group=Border_Range, parameterName=Left-RangeOffset-hover, sliderMax=1}
  n19 = ScalarParameter() {group=control, parameterName=hover, sliderMax=1}
  n20 = Multiply(
   A=Subtract(A=ComponentMask(None=@n16) {r=true}, B=LinearInterpolate(A=@n17, B=@n18, Alpha=@n19))
   B=ScalarParameter() {defaultValue=512, group=Border sharpness, parameterName=VerticalSharpness}
  )
  n21 = Rotator(
   Coordinate=TextureCoordinate() {coordinateIndex=4}
   Time=LinearInterpolate(
    A=ScalarParameter() {defaultValue=0.5025, group=Border Angles, parameterName=Bot-Angle}
    B=ScalarParameter() {defaultValue=0.4975, group=Border Angles, parameterName=Bot-Angle-hover}
    Alpha=ScalarParameter() {group=control, parameterName=hover, sliderMax=1}
   )
  ) {speed=6.2832}
  n22 = ScalarParameter() {defaultValue=0.035, group=Border_Range, parameterName=Bot-RangeOffset, sliderMax=1}
  n23 = ScalarParameter() {defaultValue=0.035, group=Border_Range, parameterName=Bot-RangeOffset-hover, sliderMax=1}
  n24 = ScalarParameter() {group=control, parameterName=hover, sliderMax=1}
  n25 = Multiply(
   A=Subtract(A=ComponentMask(None=@n21) {g=true}, B=LinearInterpolate(A=@n22, B=@n23, Alpha=@n24))
   B=ScalarParameter() {group=Border sharpness, parameterName=HorizontalSharpness}
  )
  n26 = Rotator(
   Coordinate=TextureCoordinate() {coordinateIndex=4}
   Time=LinearInterpolate(
    A=ScalarParameter() {defaultValue=0.4975, group=Border Angles, parameterName=Right-Angle}
    B=ScalarParameter() {defaultValue=0.5015, group=Border Angles, parameterName=Right-Angle-hover}
    Alpha=ScalarParameter() {group=control, parameterName=hover, sliderMax=1}
   )
  ) {speed=6.2832}
  n27 = ScalarParameter() {defaultValue=0.035, group=Border_Range, parameterName=Right-RangeOffset, sliderMax=1}
  n28 = ScalarParameter() {defaultValue=0.035, group=Border_Range, parameterName=Right-RangeOffset-hover, sliderMax=1}
  n29 = ScalarParameter() {group=control, parameterName=hover, sliderMax=1}
  n30 = Multiply(
   A=Subtract(A=ComponentMask(None=@n26) {r=true}, B=LinearInterpolate(A=@n27, B=@n28, Alpha=@n29))
   B=ScalarParameter() {group=Border sharpness, parameterName=VerticalSharpness}
  )
  n31 = DotProduct(A=DotProduct(A=Saturate(None=@n15), B=Saturate(None=@n20)), B=DotProduct(A=Saturate(None=@n25), B=Saturate(None=@n30)))
  n32 = Saturate(None=OneMinus(None=Saturate(None=@n31)))
  n33 = LinearInterpolate(
   A=ScalarParameter() {defaultValue=0.25, group=Border Alpha, parameterName=BorderAlpha, sliderMax=1}
   B=ScalarParameter() {defaultValue=1, group=Border Alpha, parameterName=HoveredBorderAlpha, sliderMax=1, sliderMin=0.01}
   Alpha=ScalarParameter() {group=control, parameterName=Selected, sliderMax=1}
  ) {constB=0}
  Add(A=@n10, B=Saturate(None=Multiply(A=@n32, B=@n33)))
  n34 = VectorParameter() {parameterName=Background_topgradient_F, defaultValue=<rgba 0 0.004 0.0723 1>}
  n35 = ScalarParameter() {group=control, parameterName=IsDiscovered, sliderMax=1}
  n36 = ComponentMask(None=TextureCoordinate()) {g=true}
  n37 = ScalarParameter() {defaultValue=0.4, group=Item, parameterName=FallOff_Placement, sliderMax=1}
  n38 = ScalarParameter() {defaultValue=1, group=Item, parameterName=FallOff_Strenght}
  n39 = ScalarParameter() {defaultValue=1, group=Item, parameterName=FallOff_Strenght_Discovered}
  n40 = ScalarParameter() {group=control, parameterName=IsDiscovered, sliderMax=1}
  n41 = Saturate(None=Multiply(A=Subtract(A=@n36, B=@n37) {constB=0.5}, B=LinearInterpolate(A=@n38, B=@n39, Alpha=@n40) {constB=0}) {constB=2})
  n42 = ScalarParameter() {group=control, parameterName=hover, sliderMax=1, sliderMin=0.1}
  LinearInterpolate(
   A=LinearInterpolate(
    A=LinearInterpolate(A=@n34.RGB, B=@n8.RGB, Alpha=@n35)
    B=VectorParameter() {group=background, parameterName=Background_topgradient_T, defaultValue=<rgba 0 0.0578 0.6038 1>}.RGB
    Alpha=LinearInterpolate(A=@n41, Alpha=@n42) {constA=1, constB=0}
   )
   B=VectorParameter() {group=ScoreIcon, parameterName=Score_RGB2, defaultValue=<rgba 1 0.7011 0.0144 1>}.RGB
   Alpha=Multiply(A=OneMinus(None=@n8.A), B=@n9)
  )
  Saturate(None=@n10)
  n43 = MaterialFunctionCall(Float2=MaterialFunctionCall() {materialFunction=&GetUserInterfaceUV}.Pixel Size) {materialFunction=&BreakOutFloat2Components}
  n44 = Max(A=@n43.G)
  n45 = MaterialFunctionCall(Float2=TextureCoordinate() {coordinateIndex=4}) {materialFunction=&BreakOutFloat2Components}
  n46 = AppendVector(A=Multiply(A=Divide(A=@n43.R, B=@n44), B=@n45.R), B=@n45.G)
  n47 = LinearInterpolate(
   A=AppendVector(
    A=ScalarParameter() {defaultValue=0.4, group=ScoreIcon, parameterName=Scoreposition_U}
    B=ScalarParameter() {defaultValue=0.38, group=ScoreIcon, parameterName=Scoreposition_V}
   )
   B=AppendVector(
    A=ScalarParameter() {defaultValue=0.4, group=ScoreIcon, parameterName=Scoreposition_U_Hover}
    B=ScalarParameter() {defaultValue=0.38, group=ScoreIcon, parameterName=Scoreposition_V_Hover}
   )
   Alpha=ScalarParameter() {group=control, parameterName=Selected, sliderMax=1}
  )
  TextureSample(UVs=MaterialFunctionCall(UVs=Add(A=@n46, B=@n47), Texture Scale=ScalarParameter() {defaultValue=0.2, group=ScoreIcon, parameterName=ScoreIconScale, sliderMax=1}) {materialFunction=&ScaleUVsByCenter}.UVs) {samplerSource=SSM_Clamp_WorldGroupSettings, texture=&/T_Ui_LocationMaker_64}
 ]
}
