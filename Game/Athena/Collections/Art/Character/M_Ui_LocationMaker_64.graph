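# UI-domain translucent material named for a 64-sized location marker icon in the Athena Collections character art.
# It holds a texture sample of T_Ui_LocationMaker_64 and a scalar constant of one.
# The graph records no links between these nodes or to the material outputs, and no parameters.
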Material /Game/Athena/Collections/Art/Character/M_Ui_LocationMaker_64
graph {
 properties {blendMode=BLEND_Translucent, materialDomain=MD_UI, pixelDepthOffsetMode=PDOM_Legacy, usedShadingModels=Default Lit}
 nodes [
  Constant() {r=1}
  TextureSample() {texture=&/T_Ui_LocationMaker_64}
 ]
}
