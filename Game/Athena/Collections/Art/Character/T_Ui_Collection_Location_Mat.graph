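# Translucent UI material for a Collections location image, built around a LocationImage texture parameter and a linear gradient.
# The gradient offsets the UVs by GradientOffset, rotates them by Vert-Angle, then scales the red channel by GradientStrenght.
# The Rotator takes a constant parameter as Time with a speed of 2π, so it acts as a static angle in turns.
# The graph shows no link from the gradient or texture sample to the material outputs.
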Material /Game/Athena/Collections/Art/Character/T_Ui_Collection_Location_Mat
graph {
 properties {blendMode=BLEND_Translucent, materialDomain=MD_UI, pixelDepthOffsetMode=PDOM_Legacy, usedShadingModels=Default Lit}
 nodes [
  n1 = Add(A=TextureCoordinate(), B=ScalarParameter() {parameterName=GradientOffset}) {constB=0}
  n2 = ScalarParameter() {defaultValue=-0.0025, group=Border Angles, parameterName=Vert-Angle}
  Multiply(
   A=ComponentMask(None=Rotator(Coordinate=@n1, Time=@n2) {speed=6.2832}) {r=true}
   B=ScalarParameter() {defaultValue=1, parameterName=GradientStrenght}
  )
  TextureSampleParameter2D() {parameterName=LocationImage, texture=&/T_Ui_Collection_Location}
 ]
}
